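; Shared Blueprint function library for Athena consumables that decides whether thrown or deployed items can spawn where they hit.
; IsHitValidForSpawn filters hits by NoStick tag, target type, surface slope and physics simulation, then flags blocking-volume hits, noted as reverting to pickup.
; Blocking uses a sphere trace for tagged components, toggled by a scalable float and skipped for exempt consumables.
; A separate helper snaps actor forward to a cardinal direction and yaw.
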(fn CheckIsBlockedByVolume
  (params (WorldContextObject :type Object%20Reference) (Location :type Vector) (BlockingAreaCheckRadius :type Float%20%28double-precision%29) (BlockingTag :type Name) (CanSpawnInBlockVolume :type Boolean))
  ;[1 1 1 0.5] Check for volumes that would block this Consumable from spawning.
  ;;[1 1 1 0.5] Check If the consumable ignores blocked volumes
  (if CanSpawnInBlockVolume
    (return false)
    (else
      (bind _boverlapping (Default|IsOverlappingDeployableBlockVolume WorldContextObject BlockingAreaCheckRadius Location BlockingTag))
      (return _boverlapping))))

(fn GetActorForwardCardinalDirection
  (params (Actor :type Actor%20Object%20Reference))
  (bind _returnvalue (Transformation|GetActorForwardVector Actor))
  (bind _x (.x _returnvalue))
  (bind _y (.y _returnvalue))
  (Utilities|FlowControl|Sequence
    (:then_0
      (Utilities|IsValid Actor
        (:Is%20Valid
          (if (>= (Math|Float|Absolute(Float) _x) (Math|Float|Absolute(Float) _y))
            (if (>= _x 0.0)
              (LocalVariables|SetCardinalDirection "NewEnumerator0")
              (LocalVariables|SetYaw 0.0)
              (else
                (LocalVariables|SetCardinalDirection "NewEnumerator1")
                (LocalVariables|SetYaw 1.8e+02)))
            (else
              (if (>= _y 0.0)
                (LocalVariables|SetCardinalDirection "NewEnumerator2")
                (LocalVariables|SetYaw 9e+01)
                (else
                  (LocalVariables|SetCardinalDirection "NewEnumerator3")
                  (LocalVariables|SetYaw 2.7e+02))))))))
    (:then_1
      (return (LocalVariables|GetCardinalDirection) (LocalVariables|GetYaw)))))

(fn IsHitValidForSpawn
  (params (HitOtherComp :type Primitive%20Component%20Object%20Reference) (HitOther :type Actor%20Object%20Reference) (HitResult :type Hit%20Result%20Structure) (SpawnOnCollision :type Boolean) (DontSpawnOnBGAWithGravity :type Scalable%20Float%20Structure) (DontSpawnOnPlayers :type Boolean) (DontSpawnOnVehicles :type Boolean) (DontSpawnOnUnderneathHit :type Boolean) (MinDotWithUp :type Float%20%28double-precision%29) (DontSpawnOnPhysicsSim :type Boolean) (CanSpawnInBlockVolume :type Boolean) (WorldContextObject :type Object%20Reference) (BlockingAreaCheckRadius :type Float%20%28double-precision%29) (BlockingVolumeTag :type Name))
  (bind (_bblockinghit _binitialoverlap _time _distance _location _impactpoint _normal _impactnormal _physmat _hitactor) (Collision|BreakHitResult HitResult))
  (Utilities|IsValid HitOtherComp
    (:Is%20Valid
      (Utilities|IsValid HitOther
        (:Is%20Valid
          ;[0.474 0.6502 1] Can stick to the hit target and has is set to spawn on collision
          (if (and (not (Components|ComponentHasTag HitOtherComp "NoStick")) (Variables|GetSpawnOnCollision))
            (Utilities|Casting|CastToBGA_Athena_WithGravity_Parent HitOther
              (:then
                ;[0.474 0.6502 1] Won't spawn on objects that can fall
                (if (not (and (Class|AthenaFunctionLibrary|RowtoBool (Variables|GetDontSpawnOnBGAWithGravity)) false))
                  ;[0.474 0.6502 1] Won't Spawn on held objects
                  (bind _returnvalue (GameplayTags|HasMatchingGameplayTag(Message) HitOther "(TagName=\"HeldObject\")"))
                  (if (not _returnvalue)
                    ;[0.474 0.6502 1] Check if it should spawn on players
                    (Utilities|Casting|CastToFortPlayerPawn HitOther
                      (:then
                        (if (not (and (Variables|GetDontSpawnOnPlayers) false))
                          ;[0.474 0.6502 1] If spawning on vehicles is allowed, check if vehicle
                          (Utilities|Casting|CastToFortAthenaVehicle HitOther
                            (:then
                              (if (and (Variables|GetDontSpawnOnVehicles) false)
                                _
                                (elif (and (Variables|GetDontSpawnOnUnderneathHit) (not (> (Math|Vector|DotProduct _normal "0, 0,1.000000") (Variables|GetMinDotWithUp))))
                                  _
                                  (else
                                    ;[0.474 0.6502 1] Don't land on physics based objects, since those can move via the impact
                                    (Utilities|IsValid _hitactor
                                      (:Is%20Valid
                                        (bind _returnvalue_2 (FortPhysicsSimulationLibrary|CanActorSimulatePhysics _hitactor))
                                        (if (not (and _returnvalue_2 (Variables|GetDontSpawnOnPhysicsSim)))
                                          ;[0.474 0.6502 1] Check if trying to spawn in a blocking volume (will revert to pickup)
                                          ; Thrown objects can still be spawned based on CanSpawnInBlockVolume bool
                                          (bind _isblocked (CallFunction|CheckIsBlockedbyVolume (Variables|GetWorldContextObject) _location (Variables|GetBlockingAreaCheckRadius) (Variables|GetBlockingVolumeTag) (Variables|GetCanSpawnInBlockVolume)))
                                          (if _isblocked
                                            (return false true)
                                            (else
                                              (return true false)))))
                                      (:Is%20Not%20Valid
                                        _))
                                    (return false false)))))
                            (:CastFailed
                              _))))
                      (:CastFailed
                        _)))))
              (:CastFailed
                _))
            (else
              _)))
        (:Is%20Not%20Valid
          _)))
    (:Is%20Not%20Valid
      _)))

(fn IsOverlappingDeployableBlockVolume
  (params (WorldContextObject :type Object%20Reference) (BlockingAreaCheckRadius :type Float%20%28single-precision%29) (Location :type Vector) (ItemSpecificBlockingTag :type Name))
  (if (ScalableFloat|GetValueasBool (LocalVariables|GetbBlockThrowableDeployVolumesEnabled))
    (Utilities|IsValid WorldContextObject
      (:Is%20Valid
        (bind _outhits (Collision|MultiSphereTraceForObjects WorldContextObject Location Location BlockingAreaCheckRadius (Utilities|Array|MakeArray "ObjectTypeQuery1" "ObjectTypeQuery2")))
        (for _array_element _outhits
          (Utilities|IsValid (.HitComponent (Collision|BreakHitResult _array_element))
            (:Is%20Valid
              (if (or (Components|ComponentHasTag (.HitComponent (Collision|BreakHitResult _array_element)) (LocalVariables|GetBlockedTag)) (Components|ComponentHasTag (.HitComponent (Collision|BreakHitResult _array_element)) ItemSpecificBlockingTag))
                (return true)))))
        (return false))
      (:Is%20Not%20Valid
        _))
    (else
      _))
  (return false))
